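; Player pawn Blueprint for Realm, an AI-controlled Fortnite character (a comment names Vader), forced into a walking style with NPC health bars.
; The server spawns a start-location actor for the Blackboard, destroys it on death, and tracks the AI target.
; On clients, the target's RepNotify runs a timer computing signed yaw for animation.
; It also flags timed deflect poses, records holstered state for death animation, and fades breathing audio on death.
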(event EventBeginPlay
  ;[1 1 1] Check to know if the AI has a Weapon Equipped or not at the time of death (before he actually dies - when he dies, it removes the weapon). Set a variable that will be used in the Gamepaly Cue Rezzout to validate which animation should play.
  (|CallParentFunction)
  (FortPawn|BindEventtoOnWeaponUnEquippedDelegate (delegate Custom|WeaponUnequipped))
  (FortPawn|BindEventtoOnWeaponEquippedDelegate (delegate Custom|WeaponEquipped))
  ;[1 1 1] Force walking movement style
  (Class|FortPlayerAnimationInterface|SetMovementStyleOverride (Animation|GetAnimationInterface self) "Walking")
  (Networking|SwitchHasAuthority
    (:Authority
      (Utilities|FlowControl|Sequence
        (:then_0
          ;[1 1 1] To Be Used in the Ability where Realm needs to Jump away from Players if Realm and Player is in the Water.
          (Variables|Default|SetSpawnLocation (Transformation|GetActorLocation self))
          (bind _returnvalue (Game|SpawnActorfromClass "/Realm/Items/BP_Realm_StartLocation.BP_Realm_StartLocation_C" (Transformation|GetActorTransform self) "AlwaysSpawn" "OverrideRootScale" self self))
          (Variables|Default|SetRealmOriginActor _returnvalue)
          (Utilities|FlowControl|Delay 1.0) ; This 1s delay is needed, else the Blackboard doesn't seem ready to receive a Value as Object
          (AI|Components|Blackboard|SetValueasObject (AI|GetBlackboard (AI|GetAIController self)) (Variables|Default|GetBBKEY_OriginActor) (Variables|Default|GetRealmOriginActor))
          (Variables|Default|SetCanEnableHealthBars true))
        (:then_1
          (bind _asfort_athena_aibot_controller (Utilities|Casting|CastToFortAthenaAIBotController (Pawn|GetController self)))
          (Perception|BindEventtoOnCurrentTargetChangedEvent _asfort_athena_aibot_controller (delegate Custom|CurrentTargetChanged)))))
    (:Remote
      (Utilities|FlowControl|Delay 1.0)
      (CallFunction|NPCEnableHealthBars) ; Make sure Health Bars are Being Displayed
    )))

(event Custom|WeaponUnequipped
  ;[1 1 1] Check to know if the AI has a Weapon Equipped or not at the time of death (before he actually dies - when he dies, it removes the weapon). Set a variable that will be used in the Gamepaly Cue Rezzout to validate which animation should play.
  (if (not (FortPawn|IsDead self))
    (Variables|Default|SetWeaponHolsteredOnDeath true)))

(event Custom|WeaponEquipped
  (params (NewWeapon :type Fort%20Weapon%20Object%20Reference) (PrevWeapon :type Fort%20Weapon%20Object%20Reference))
  ;[1 1 1] Check to know if the AI has a Weapon Equipped or not at the time of death (before he actually dies - when he dies, it removes the weapon). Set a variable that will be used in the Gamepaly Cue Rezzout to validate which animation should play.
  (if (not (FortPawn|IsDead self))
    (Variables|Default|SetWeaponHolsteredOnDeath false)))

(event Custom|CurrentTargetChanged
  (params (OldTarget :type Actor%20Object%20Reference) (NewTarget :type Actor%20Object%20Reference))
  (Variables|Default|SetAITarget NewTarget))

(event Custom|RealmIsDeflecting
  ;[1 1 1] Realm is Deflecting - Specific Bool for Anim Tree for Animators to get Vader to Hold a Pose
  (Variables|Default|SetIsCurrentlyDeflecting true)
  (Networking|ForceNetUpdate)
  (Utilities|FlowControl|RetriggerableDelay (ScalableFloat|GetValueatLevel (Variables|Default|GetDeflectingAnimDelay)))
  (Variables|Default|SetIsCurrentlyDeflecting false)
  (Networking|ForceNetUpdate))

(event Custom|StartYawUpdateTimer
  ;[1 1 1] Check if Pawn has a Target and get the Angle Difference and setit in a Variabl so Animators can use the value.
  (bind _returnvalue (Utilities|Time|SetTimerbyEvent (EventDispatchers|CreateEvent :!event Custom|CheckPawnTargetPosition) 0.1 true))
  (Variables|InternalData|SetHNDL_UpdateTargettingYaw _returnvalue))

(event EventOnDeathServer
  (params (Damage :type Float%20%28single-precision%29) (DamageTags :type Gameplay%20Tag%20Container%20Structure%20%28by%20ref%29) (Momentum :type Vector :default "0, 0, 0") (HitInfo :type Hit%20Result%20Structure%20%28by%20ref%29) (InstigatedBy :type Controller%20Object%20Reference) (DamageCauser :type Actor%20Object%20Reference) (EffectContext :type Gameplay%20Effect%20Context%20Handle%20Structure))
  (bind _realmoriginactor (Variables|Default|GetRealmOriginActor))
  ;[1 1 1] On Death Server - Destroy the Origin Actor we Spawned
  (Utilities|IsValid _realmoriginactor
    (:Is%20Valid
      (Actor|DestroyActor _realmoriginactor))))

(event Custom|CheckPawnTargetPosition
  (bind _aitarget (Variables|Default|GetAITarget))
  ;[1 1 1] Check if Pawn has a Target and get the Angle Difference and setit in a Variabl so Animators can use the value.
  (Utilities|IsValid _aitarget
    (:Is%20Valid
      (Variables|Animation|SetTargetYaw (select (< (Math|Vector|DotProduct (Transformation|GetActorRightVector self) (Math|Vector|Normalize (- (Transformation|GetActorLocation _aitarget) (Transformation|GetActorLocation self)))) 0) (Class|AbilityFunctions|GetAngleBetweenTwoVectors (Math|Vector|Normalize (- (Transformation|GetActorLocation _aitarget) (Transformation|GetActorLocation self))) (Transformation|GetActorForwardVector self)) (* (Class|AbilityFunctions|GetAngleBetweenTwoVectors (Math|Vector|Normalize (- (Transformation|GetActorLocation _aitarget) (Transformation|GetActorLocation self))) (Transformation|GetActorForwardVector self)) -1.0))))
    (:Is%20Not%20Valid
      (Variables|Animation|SetTargetYaw 0.0) ; If no Target, set the Yaw at 0
    )))

(event EventOnDeathPlayEffects
  (params (Damage :type Float%20%28single-precision%29) (DamageTags :type Gameplay%20Tag%20Container%20Structure%20%28by%20ref%29) (Momentum :type Vector :default "0, 0, 0") (HitInfo :type Hit%20Result%20Structure%20%28by%20ref%29) (InstigatedBy :type Fort%20Pawn%20Object%20Reference) (DamageCauser :type Actor%20Object%20Reference) (EffectContext :type Gameplay%20Effect%20Context%20Handle%20Structure))
  ;[1 1 1] Turn off Breathing Audio
  (|CallParentFunction Damage DamageTags Momentum HitInfo InstigatedBy DamageCauser EffectContext)
  (Audio|Components|Audio|FadeOut (Variables|Default|GetAudio%20Breathing%20Loop) 0.5 0.0 "Linear" :!class /Script/Engine.AudioComponent))

(event Collision|EventActorBeginOverlap
  (params (OtherActor :type Actor%20Object%20Reference))
  (|CallParentFunction OtherActor) ; This node is disabled and will not be called.\nDrag off pins to build functionality.
)

(event EventTick
  (params (DeltaSeconds :type Float%20%28single-precision%29))
  (|CallParentFunction DeltaSeconds) ; This node is disabled and will not be called.\nDrag off pins to build functionality.
)

(fn OnRep_AITarget
  (bind _hndl_updatetargettingyaw (Variables|InternalData|GetHNDL_UpdateTargettingYaw))
  (Networking|SwitchHasAuthority
    (:Remote
      (Utilities|IsValid (Variables|Default|GetAITarget)
        (:Is%20Valid
          (if (not (Utilities|Time|IsValidTimerHandle _hndl_updatetargettingyaw))
            (CallFunction|StartYawUpdateTimer)))
        (:Is%20Not%20Valid
          (Utilities|Time|ClearandInvalidateTimerbyHandle _hndl_updatetargettingyaw))))))

(fn OnRep_CanEnableHealthBars
  (if (Variables|Default|GetCanEnableHealthBars)
    (CallFunction|NPCEnableHealthBars)))

(fn OnRep_IsCurrentlyDeflecting
  (Networking|SwitchHasAuthority))

(fn OnRep_WeaponHolsteredOnDeath)

(fn ConstructionScript
  (|CallParentFunction))
